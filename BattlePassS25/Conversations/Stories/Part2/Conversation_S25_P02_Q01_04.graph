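# Season 25 Story Part 2 conversation database for Triggerfish's questline about the missing Coral Buddies.
# Four entry points share one tag, each opening a linear speech chain for a different quest stage.
# Opening lines are gated by an active-quest check, an inverted completed-quest check on the preceding step, and an NPC conversation tag.
# Each chain's final Continue choice carries a stage tag and leads to completing that quest's objective by backend name.
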ConversationDatabase /BattlePassS25/Conversations/Stories/Part2/Conversation_S25_P02_Q01_04
entry_points [{
  tag=Conversation.Entry.Season25.P01.Q01.Triggerfish
  destinations [#1, #2, #3, #4]
 }]
nodes [
 ConversationEntryPointNode_4 {
  id=#4
  class=ConversationEntryPointNode
  outputs [#5]
  properties {entryTag {tagName=Conversation.Entry.Season25.P01.Q01.Triggerfish}}
 }
 Speech_C_20 {
  id=#5
  class=Speech_C
  outputs [#6]
  properties {
   softTaskIcon=&T_UI_NPC_MainQuest
   generalConfig {defaultMessage="Well done there, anchovy. Or should I just call you Looper, since you're no small fish!"}
  }
  sub_nodes [
   ConversationChoiceNode_21 {
    class=ConversationChoiceNode
    properties {
     defaultChoiceDisplayText=Talk to Triggerfish
     choiceTags {gameplayTags [{tagName=Conversation.Choice.AutoSelectIfValid}, {tagName=Conversation.Choice.Category.MainQuestline}]}
    }
   }
   FortConversationRequirement_ControllerRequirement_12 {
    class=FortConversationRequirement_ControllerRequirement
    properties {
     nodeName=Require: Has Completed Quest
     requirement {
      class=FortControllerRequirement_HasCompletedQuest
      properties {bInverseCondition=true, questDefinition=/BattlePassS25/Items/QuestItems/Stories/Part02/Quest_04/quest_s25_stories_p02_q04_04}
     }
    }
   }
   FortConversationRequirement_ControllerRequirement_13 {
    class=FortConversationRequirement_ControllerRequirement
    properties {
     nodeName=Require: Is Quest Active
     requirement {
      class=FortControllerRequirement_IsQuestActive
      properties {questDefinition=/BattlePassS25/Items/QuestItems/Stories/Part02/Quest_04/quest_s25_stories_p02_q04_05}
     }
    }
   }
   FortConversationRequirement_ControllerRequirement_14 {
    class=FortConversationRequirement_ControllerRequirement
    properties {
     nodeName=Require: NPC Has Conversation Tag
     participantId {tagName=Conversation.Participant.NPC}
     requirement {class=FortControllerRequirement_NPCHasConversationEntryTag, properties {queryTag {tagName=Conversation.Entry.Trigger}}}
    }
   }
  ]
 }
 Speech_C_19 {
  id=#6
  class=Speech_C
  outputs [#7]
  properties {generalConfig {defaultMessage=I promised Princess Bryne that I'll protect these little Coral Buddies from any more fishy business.}}
  sub_nodes [ConversationChoiceNode_20 {class=ConversationChoiceNode, properties {defaultChoiceDisplayText=Continue}}]
 }
 Speech_C_17 {
  id=#7
  class=Speech_C
  outputs [#8]
  properties {generalConfig {defaultMessage=We can build 'em their own little sanctuary. I've got a cousin I can call about that...}}
  sub_nodes [ConversationChoiceNode_18 {class=ConversationChoiceNode, properties {defaultChoiceDisplayText=Continue}}]
 }
 Speech_C_18 {
  id=#8
  class=Speech_C
  outputs [#9]
  properties {generalConfig {defaultMessage="Thanks, Looper. I'll keep you in mind if any more hero business comes along down the river."}}
  sub_nodes [ConversationChoiceNode_19 {
    class=ConversationChoiceNode
    properties {defaultChoiceDisplayText=Continue, choiceTags {gameplayTags [{tagName=Athena.Quests.S25.Conversations.Triggerfish.Choice.Q04}]}}
   }]
 }
 FortConversationTaskNode_CompleteQuestObjectiveByBackendName_4 {
  id=#9
  class=FortConversationTaskNode_CompleteQuestObjectiveByBackendName
  properties {
   objectiveBackendName=quest_s25_stories_p02_q04_05_obj1
   questDefinition=&quest_s25_stories_p02_q04_05
  }
 }
 ConversationEntryPointNode_2 {
  id=#3
  class=ConversationEntryPointNode
  outputs [#10]
  properties {entryTag {tagName=Conversation.Entry.Season25.P01.Q01.Triggerfish}}
 }
 Speech_C_2 {
  id=#10
  class=Speech_C
  outputs [#11]
  properties {
   softTaskIcon=&T_UI_NPC_MainQuest
   generalConfig {defaultMessage="Looks like you've found one of our missing Coral Buddies! Good work."}
  }
  sub_nodes [
   ConversationChoiceNode_12 {
    class=ConversationChoiceNode
    properties {
     defaultChoiceDisplayText=Talk to Triggerfish
     choiceTags {gameplayTags [{tagName=Conversation.Choice.AutoSelectIfValid}, {tagName=Conversation.Choice.Category.MainQuestline}]}
    }
   }
   FortConversationRequirement_ControllerRequirement_9 {
    class=FortConversationRequirement_ControllerRequirement
    properties {
     nodeName=Require: Has Completed Quest
     requirement {
      class=FortControllerRequirement_HasCompletedQuest
      properties {bInverseCondition=true, questDefinition=/BattlePassS25/Items/QuestItems/Stories/Part02/Quest_03/quest_s25_stories_p02_q03_02}
     }
    }
   }
   FortConversationRequirement_ControllerRequirement_10 {
    class=FortConversationRequirement_ControllerRequirement
    properties {
     nodeName=Require: Is Quest Active
     requirement {
      class=FortControllerRequirement_IsQuestActive
      properties {questDefinition=/BattlePassS25/Items/QuestItems/Stories/Part02/Quest_03/quest_s25_stories_p02_q03_03}
     }
    }
   }
   FortConversationRequirement_ControllerRequirement_11 {
    class=FortConversationRequirement_ControllerRequirement
    properties {
     nodeName=Require: NPC Has Conversation Tag
     participantId {tagName=Conversation.Participant.NPC}
     requirement {class=FortControllerRequirement_NPCHasConversationEntryTag, properties {queryTag {tagName=Conversation.Entry.Trigger}}}
    }
   }
  ]
 }
 Speech_C_6 {
  id=#11
  class=Speech_C
  outputs [#12]
  properties {generalConfig {defaultMessage=This smells like Turk and Riptide. Those anglers have always chilled my gills.}}
  sub_nodes [ConversationChoiceNode_13 {class=ConversationChoiceNode, properties {defaultChoiceDisplayText=Continue}}]
 }
 Speech_C_22 {
  id=#12
  class=Speech_C
  outputs [#13]
  properties {generalConfig {defaultMessage="Turk's nice, but not good. And Riptide treats the sea like his own personal bucket o' chowder."}}
  sub_nodes [ConversationChoiceNode_23 {class=ConversationChoiceNode, properties {defaultChoiceDisplayText=Continue}}]
 }
 Speech_C_21 {
  id=#13
  class=Speech_C
  outputs [#14]
  properties {generalConfig {defaultMessage=We need 'em to tell us where the Coral Buddies are. You'll need to challenge one of em to a duel.}}
  sub_nodes [ConversationChoiceNode_22 {class=ConversationChoiceNode, properties {defaultChoiceDisplayText=Continue}}]
 }
 Speech_C_23 {
  id=#14
  class=Speech_C
  outputs [#15]
  properties {generalConfig {defaultMessage="You're the Coral Buddies' only hope, anchovy. It's time to step up and get it done."}}
  sub_nodes [ConversationChoiceNode_24 {
    class=ConversationChoiceNode
    properties {defaultChoiceDisplayText=Continue, choiceTags {gameplayTags [{tagName=Athena.Quests.S25.Conversations.Triggerfish.Choice.Q03}]}}
   }]
 }
 FortConversationTaskNode_CompleteQuestObjectiveByBackendName_5 {
  id=#15
  class=FortConversationTaskNode_CompleteQuestObjectiveByBackendName
  properties {
   objectiveBackendName=quest_s25_stories_p02_q03_03_obj2
   questDefinition=&quest_s25_stories_p02_q03_03
  }
 }
 ConversationEntryPointNode_1 {
  id=#2
  class=ConversationEntryPointNode
  outputs [#16]
  properties {entryTag {tagName=Conversation.Entry.Season25.P01.Q01.Triggerfish}}
 }
 Speech_C_4 {
  id=#16
  class=Speech_C
  outputs [#17]
  properties {
   softTaskIcon=&T_UI_NPC_MainQuest
   generalConfig {defaultMessage=These scans might be the key to finding the Coral Buddies.}
  }
  sub_nodes [
   ConversationChoiceNode_2 {
    class=ConversationChoiceNode
    properties {
     defaultChoiceDisplayText=Talk to Triggerfish
     choiceTags {gameplayTags [{tagName=Conversation.Choice.AutoSelectIfValid}, {tagName=Conversation.Choice.Category.MainQuestline}]}
    }
   }
   FortConversationRequirement_ControllerRequirement_6 {
    class=FortConversationRequirement_ControllerRequirement
    properties {
     nodeName=Require: Has Completed Quest
     requirement {
      class=FortControllerRequirement_HasCompletedQuest
      properties {bInverseCondition=true, questDefinition=/BattlePassS25/Items/QuestItems/Stories/Part02/Quest_02/quest_s25_stories_p02_q02_03}
     }
    }
   }
   FortConversationRequirement_ControllerRequirement_7 {
    class=FortConversationRequirement_ControllerRequirement
    properties {
     nodeName=Require: Is Quest Active
     requirement {
      class=FortControllerRequirement_IsQuestActive
      properties {questDefinition=/BattlePassS25/Items/QuestItems/Stories/Part02/Quest_02/quest_s25_stories_p02_q02_04}
     }
    }
   }
   FortConversationRequirement_ControllerRequirement_8 {
    class=FortConversationRequirement_ControllerRequirement
    properties {
     nodeName=Require: NPC Has Conversation Tag
     participantId {tagName=Conversation.Participant.NPC}
     requirement {class=FortControllerRequirement_NPCHasConversationEntryTag, properties {queryTag {tagName=Conversation.Entry.Trigger}}}
    }
   }
  ]
 }
 Speech_C_5 {
  id=#17
  class=Speech_C
  outputs [#18]
  properties {generalConfig {defaultMessage=Excluding the spots where rare Island coral grows... the scans pinged one spot in particular.}}
  sub_nodes [ConversationChoiceNode_4 {class=ConversationChoiceNode, properties {defaultChoiceDisplayText=Continue}}]
 }
 Speech_C_7 {
  id=#18
  class=Speech_C
  outputs [#19]
  properties {generalConfig {defaultMessage=Go fish for clues around Berg Barge. Our little Coral Buddies might be stranded there somehow.}}
  sub_nodes [ConversationChoiceNode_9 {
    class=ConversationChoiceNode
    properties {defaultChoiceDisplayText=Continue, choiceTags {gameplayTags [{tagName=Athena.Quests.S25.Conversations.Triggerfish.Choice.Q02}]}}
   }]
 }
 FortConversationTaskNode_CompleteQuestObjectiveByBackendName_0 {
  id=#19
  class=FortConversationTaskNode_CompleteQuestObjectiveByBackendName
  properties {
   objectiveBackendName=quest_s25_stories_p02_q02_04_obj3
   questDefinition=&quest_s25_stories_p02_q02_04
  }
 }
 ConversationEntryPointNode_0 {
  id=#1
  class=ConversationEntryPointNode
  outputs [#20]
  properties {entryTag {tagName=Conversation.Entry.Season25.P01.Q01.Triggerfish}}
 }
 Speech_C_3 {
  id=#20
  class=Speech_C
  outputs [#21]
  properties {
   softTaskIcon=&T_UI_NPC_MainQuest
   generalConfig {defaultMessage="You've impressed me, anchovy. Didn't think ya had it in you."}
  }
  sub_nodes [
   ConversationChoiceNode_5 {
    class=ConversationChoiceNode
    properties {
     defaultChoiceDisplayText=Talk to Triggerfish
     choiceTags {gameplayTags [{tagName=Conversation.Choice.AutoSelectIfValid}, {tagName=Conversation.Choice.Category.MainQuestline}]}
    }
   }
   FortConversationRequirement_ControllerRequirement_2 {
    class=FortConversationRequirement_ControllerRequirement
    properties {
     nodeName=Require: Has Completed Quest
     requirement {
      class=FortControllerRequirement_HasCompletedQuest
      properties {bInverseCondition=true, questDefinition=/BattlePassS25/Items/QuestItems/Stories/Part02/Quest_01/quest_s25_stories_p02_q03}
     }
    }
   }
   FortConversationRequirement_ControllerRequirement_4 {
    class=FortConversationRequirement_ControllerRequirement
    properties {
     nodeName=Require: Is Quest Active
     requirement {
      class=FortControllerRequirement_IsQuestActive
      properties {questDefinition=/BattlePassS25/Items/QuestItems/Stories/Part02/Quest_01/quest_s25_stories_p02_q04}
     }
    }
   }
   FortConversationRequirement_ControllerRequirement_5 {
    class=FortConversationRequirement_ControllerRequirement
    properties {
     nodeName=Require: NPC Has Conversation Tag
     participantId {tagName=Conversation.Participant.NPC}
     requirement {class=FortControllerRequirement_NPCHasConversationEntryTag, properties {queryTag {tagName=Conversation.Entry.Trigger}}}
    }
   }
  ]
 }
 Speech_C_28 {
  id=#21
  class=Speech_C
  outputs [#22]
  properties {generalConfig {defaultMessage=I've got a mission for you. The Coral Buddies have vanished. And something smells... fishy.}}
  sub_nodes [ConversationChoiceNode_6 {class=ConversationChoiceNode, properties {defaultChoiceDisplayText=Continue}}]
 }
 Speech_C_0 {
  id=#22
  class=Speech_C
  outputs [#23]
  properties {generalConfig {defaultMessage="I need your help investigating. This is a big Island, and those Coral Buddies are awful small."}}
  sub_nodes [ConversationChoiceNode_7 {class=ConversationChoiceNode, properties {defaultChoiceDisplayText=Continue}}]
 }
 Speech_C_1 {
  id=#23
  class=Speech_C
  outputs [#24]
  properties {generalConfig {defaultMessage=Princess Bryne's the one who asked me to track down the Coral Buddies. She might have an idea where to look.}}
  sub_nodes [ConversationChoiceNode_8 {
    class=ConversationChoiceNode
    properties {defaultChoiceDisplayText=Continue, choiceTags {gameplayTags [{tagName=Athena.Quests.S25.Conversations.Triggerfish.Choice.Q01}]}}
   }]
 }
 FortConversationTaskNode_CompleteQuestObjectiveByBackendName_1 {
  id=#24
  class=FortConversationTaskNode_CompleteQuestObjectiveByBackendName
  properties {
   objectiveBackendName=quest_s25_stories_p02_q04_obj1
   questDefinition=&quest_s25_stories_p02_q04
  }
 }
]
