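# Season 25 NPC conversation in which Bryne, worried about the missing Coral Buddies, sends the player to gather rare coral for tracking them.
# The opening line auto-selects when this quest is active, quest p02_q04 is incomplete and the NPC has Bryne's entry tag.
# A linear chain of speech nodes ends by completing objective quest_s25_stories_p02_q02_01_obj1.
# The entry tag reads P01.Q01 although the asset belongs to Part 2, Quest 2.
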ConversationDatabase /BattlePassS25/Conversations/Stories/Part2/Conversation_S25_P02_Q02_01
entry_points [{tag=Conversation.Entry.Season25.P01.Q01.Bryne, destinations [#1]}]
nodes [
 ConversationEntryPointNode_0 {
  id=#1
  class=ConversationEntryPointNode
  outputs [#2]
  properties {entryTag {tagName=Conversation.Entry.Season25.P01.Q01.Bryne}}
 }
 Speech_C_3 {
  id=#2
  class=Speech_C
  outputs [#3]
  properties {
   softTaskIcon=&T_UI_NPC_MainQuest
   generalConfig {defaultMessage="Triggerfish sent you? Oh, thank the Great Kraken."}
  }
  sub_nodes [
   ConversationChoiceNode_0 {
    class=ConversationChoiceNode
    properties {
     defaultChoiceDisplayText=Talk to Bryne
     choiceTags {gameplayTags [{tagName=Conversation.Choice.AutoSelectIfValid}, {tagName=Conversation.Choice.Category.MainQuestline}]}
    }
   }
   FortConversationRequirement_ControllerRequirement_0 {
    class=FortConversationRequirement_ControllerRequirement
    properties {
     nodeName=Require: Has Completed Quest
     requirement {
      class=FortControllerRequirement_HasCompletedQuest
      properties {bInverseCondition=true, questDefinition=/BattlePassS25/Items/QuestItems/Stories/Part02/Quest_01/quest_s25_stories_p02_q04}
     }
    }
   }
   FortConversationRequirement_ControllerRequirement_1 {
    class=FortConversationRequirement_ControllerRequirement
    properties {
     nodeName=Require: Is Quest Active
     requirement {
      class=FortControllerRequirement_IsQuestActive
      properties {questDefinition=/BattlePassS25/Items/QuestItems/Stories/Part02/Quest_02/quest_s25_stories_p02_q02_01}
     }
    }
   }
   FortConversationRequirement_ControllerRequirement_3 {
    class=FortConversationRequirement_ControllerRequirement
    properties {
     nodeName=Require: NPC Has Conversation Tag
     participantId {tagName=Conversation.Participant.NPC}
     requirement {class=FortControllerRequirement_NPCHasConversationEntryTag, properties {queryTag {tagName=Conversation.Entry.Bryne}}}
    }
   }
  ]
 }
 Speech_C_28 {
  id=#3
  class=Speech_C
  outputs [#4]
  properties {generalConfig {defaultMessage=I'm so worried about our helpless little Coral Buddies. If something bad has happened to them...}}
  sub_nodes [ConversationChoiceNode_1 {class=ConversationChoiceNode, properties {defaultChoiceDisplayText=Continue}}]
 }
 Speech_C_0 {
  id=#4
  class=Speech_C
  outputs [#5]
  properties {generalConfig {defaultMessage=Now.. The Coral Buddies share genetic markers with rare Island coral. Could we use that to track them?}}
  sub_nodes [ConversationChoiceNode_2 {class=ConversationChoiceNode, properties {defaultChoiceDisplayText=Continue}}]
 }
 Speech_C_1 {
  id=#5
  class=Speech_C
  outputs [#6]
  properties {generalConfig {defaultMessage="There is little island off the south west coast where the rare coral grows. Please, go gather some."}}
  sub_nodes [ConversationChoiceNode_3 {class=ConversationChoiceNode, properties {defaultChoiceDisplayText=Continue}}]
 }
 Speech_C_2 {
  id=#6
  class=Speech_C
  outputs [#7]
  properties {generalConfig {defaultMessage="I will secure us some equipment to analyse the coral. Now hurry, please!"}}
  sub_nodes [ConversationChoiceNode_4 {
    class=ConversationChoiceNode
    properties {defaultChoiceDisplayText=Continue, choiceTags {gameplayTags [{tagName=Athena.Quests.S25.Conversations.Bryne.Choice.Q01}]}}
   }]
 }
 FortConversationTaskNode_CompleteQuestObjectiveByBackendName_1 {
  id=#7
  class=FortConversationTaskNode_CompleteQuestObjectiveByBackendName
  properties {
   objectiveBackendName=quest_s25_stories_p02_q02_01_obj1
   questDefinition=&quest_s25_stories_p02_q02_01
  }
 }
]
