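# Conversation database for Season 25 Story Part 2, where players confront Turk and Riptide about the missing Coral Buddies.
# Each NPC has two branches under one entry tag: denial or threats, and surrendering the cave location.
# Auto-selected openers are gated by NPC tag, active quest and uncompleted quest: Q03 for denials, Q04 for surrenders.
# All branches end completing that NPC's Q03_02 objective, even the Q04-gated ones.
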ConversationDatabase /BattlePassS25/Conversations/Stories/Part2/Conversation_S25_P02_Q03_02
entry_points [
 {tag=Conversation.Entry.Season25.P01.Q01.Turk, destinations [#1, #2]}
 {tag=Conversation.Entry.Season25.P01.Q01.Riptide, destinations [#3, #4]}
]
nodes [
 ConversationEntryPointNode_3 {
  id=#4
  class=ConversationEntryPointNode
  outputs [#5]
  properties {entryTag {tagName=Conversation.Entry.Season25.P01.Q01.Riptide}}
 }
 Speech_C_10 {
  id=#5
  class=Speech_C
  outputs [#6]
  properties {
   softTaskIcon=&T_UI_NPC_MainQuest
   generalConfig {defaultMessage="Enough, enough. I know when to stop fishin' in dangerous waters."}
  }
  sub_nodes [
   ConversationChoiceNode_0 {
    class=ConversationChoiceNode
    properties {
     defaultChoiceDisplayText=Confront Riptide
     choiceTags {gameplayTags [{tagName=Conversation.Choice.AutoSelectIfValid}, {tagName=Conversation.Choice.Category.MainQuestline}]}
    }
   }
   FortConversationRequirement_ControllerRequirement_0 {
    class=FortConversationRequirement_ControllerRequirement
    properties {
     nodeName=Require: Has Completed Quest
     requirement {
      class=FortControllerRequirement_HasCompletedQuest
      properties {bInverseCondition=true, questDefinition=/BattlePassS25/Items/QuestItems/Stories/Part02/Quest_04/quest_s25_stories_p02_q04_02}
     }
    }
   }
   FortConversationRequirement_ControllerRequirement_1 {
    class=FortConversationRequirement_ControllerRequirement
    properties {
     nodeName=Require: Is Quest Active
     requirement {
      class=FortControllerRequirement_IsQuestActive
      properties {questDefinition=/BattlePassS25/Items/QuestItems/Stories/Part02/Quest_04/quest_s25_stories_p02_q04_03}
     }
    }
   }
   FortConversationRequirement_ControllerRequirement_3 {
    class=FortConversationRequirement_ControllerRequirement
    properties {
     nodeName=Require: NPC Has Conversation Tag
     participantId {tagName=Conversation.Participant.NPC}
     requirement {class=FortControllerRequirement_NPCHasConversationEntryTag, properties {queryTag {tagName=Conversation.Entry.Riptide}}}
    }
   }
  ]
 }
 Speech_C_12 {
  id=#6
  class=Speech_C
  outputs [#7]
  properties {generalConfig {defaultMessage=We've got those Coral Buddies holed up in a nearby cave. I'll give you the location.}}
  sub_nodes [ConversationChoiceNode_2 {
    class=ConversationChoiceNode
    properties {defaultChoiceDisplayText=Continue, choiceTags {gameplayTags [{tagName=Athena.Quests.S25.Conversations.Riptide.Choice.Q02}]}}
   }]
 }
 FortConversationTaskNode_CompleteQuestObjectiveByBackendName_3 {
  id=#7
  class=FortConversationTaskNode_CompleteQuestObjectiveByBackendName
  properties {
   objectiveBackendName=quest_s25_stories_p02_q03_02_obj3
   questDefinition=/BattlePassS25/Items/QuestItems/Stories/Part02/Quest_03/quest_s25_stories_p02_q03_02
  }
 }
 ConversationEntryPointNode_2 {
  id=#2
  class=ConversationEntryPointNode
  outputs [#8]
  properties {entryTag {tagName=Conversation.Entry.Season25.P01.Q01.Turk}}
 }
 Speech_C_7 {
  id=#8
  class=Speech_C
  outputs [#9]
  properties {
   softTaskIcon=&T_UI_NPC_MainQuest
   generalConfig {defaultMessage=I don't want any trouble. I'll give you the location of the Coral Buddies.}
  }
  sub_nodes [
   ConversationChoiceNode_5 {
    class=ConversationChoiceNode
    properties {
     defaultChoiceDisplayText=Confront Turk
     choiceTags {gameplayTags [{tagName=Conversation.Choice.AutoSelectIfValid}, {tagName=Conversation.Choice.Category.MainQuestline}]}
    }
   }
   FortConversationRequirement_ControllerRequirement_2 {
    class=FortConversationRequirement_ControllerRequirement
    properties {
     nodeName=Require: Has Completed Quest
     requirement {
      class=FortControllerRequirement_HasCompletedQuest
      properties {bInverseCondition=true, questDefinition=&quest_s25_stories_p02_q04_02}
     }
    }
   }
   FortConversationRequirement_ControllerRequirement_4 {
    class=FortConversationRequirement_ControllerRequirement
    properties {
     nodeName=Require: Is Quest Active
     requirement {
      class=FortControllerRequirement_IsQuestActive
      properties {questDefinition=&quest_s25_stories_p02_q04_03}
     }
    }
   }
   FortConversationRequirement_ControllerRequirement_5 {
    class=FortConversationRequirement_ControllerRequirement
    properties {
     nodeName=Require: NPC Has Conversation Tag
     participantId {tagName=Conversation.Participant.NPC}
     requirement {class=FortControllerRequirement_NPCHasConversationEntryTag, properties {queryTag {tagName=Conversation.Entry.Turk}}}
    }
   }
  ]
 }
 Speech_C_9 {
  id=#9
  class=Speech_C
  outputs [#10]
  properties {generalConfig {defaultMessage="We've been keeping the Coral Buddies in a nearby cave, safe and sound. I'll put the location on your map."}}
  sub_nodes [ConversationChoiceNode_7 {
    class=ConversationChoiceNode
    properties {defaultChoiceDisplayText=Continue, choiceTags {gameplayTags [{tagName=Athena.Quests.S25.Conversations.Turk.Choice.Q02}]}}
   }]
 }
 FortConversationTaskNode_CompleteQuestObjectiveByBackendName_2 {
  id=#10
  class=FortConversationTaskNode_CompleteQuestObjectiveByBackendName
  properties {
   objectiveBackendName=quest_s25_stories_p02_q03_02_obj2
   questDefinition=&quest_s25_stories_p02_q03_02
  }
 }
 ConversationEntryPointNode_1 {
  id=#3
  class=ConversationEntryPointNode
  outputs [#11]
  properties {entryTag {tagName=Conversation.Entry.Season25.P01.Q01.Riptide}}
 }
 Speech_C_4 {
  id=#11
  class=Speech_C
  outputs [#12]
  properties {
   softTaskIcon=&T_UI_NPC_MainQuest
   generalConfig {defaultMessage=You look like you're poking around where you don't belong...}
  }
  sub_nodes [
   ConversationChoiceNode_11 {
    class=ConversationChoiceNode
    properties {
     defaultChoiceDisplayText=Confront Riptide
     choiceTags {gameplayTags [{tagName=Conversation.Choice.AutoSelectIfValid}, {tagName=Conversation.Choice.Category.MainQuestline}]}
    }
   }
   FortConversationRequirement_ControllerRequirement_9 {
    class=FortConversationRequirement_ControllerRequirement
    properties {
     nodeName=Require: Has Completed Quest
     requirement {
      class=FortControllerRequirement_HasCompletedQuest
      properties {bInverseCondition=true, questDefinition=/BattlePassS25/Items/QuestItems/Stories/Part02/Quest_03/quest_s25_stories_p02_q03_01}
     }
    }
   }
   FortConversationRequirement_ControllerRequirement_10 {
    class=FortConversationRequirement_ControllerRequirement
    properties {
     nodeName=Require: Is Quest Active
     requirement {
      class=FortControllerRequirement_IsQuestActive
      properties {questDefinition=&quest_s25_stories_p02_q03_02}
     }
    }
   }
   FortConversationRequirement_ControllerRequirement_11 {
    class=FortConversationRequirement_ControllerRequirement
    properties {
     nodeName=Require: NPC Has Conversation Tag
     participantId {tagName=Conversation.Participant.NPC}
     requirement {class=FortControllerRequirement_NPCHasConversationEntryTag, properties {queryTag {tagName=Conversation.Entry.Riptide}}}
    }
   }
  ]
 }
 Speech_C_5 {
  id=#12
  class=Speech_C
  outputs [#13]
  properties {generalConfig {defaultMessage="You better head downstream, little fish. Or you might get eaten by sharks."}}
  sub_nodes [ConversationChoiceNode_12 {class=ConversationChoiceNode, properties {defaultChoiceDisplayText=Continue}}]
 }
 Speech_C_14 {
  id=#13
  class=Speech_C
  outputs [#14]
  properties {generalConfig {defaultMessage="Besides, who cares about those little Coral Buddies anyway?"}}
  sub_nodes [ConversationChoiceNode_15 {
    class=ConversationChoiceNode
    properties {defaultChoiceDisplayText=Continue, choiceTags {gameplayTags [{tagName=Athena.Quests.S25.Conversations.Riptide.Choice.Q01}]}}
   }]
 }
 FortConversationTaskNode_CompleteQuestObjectiveByBackendName_6 {
  id=#14
  class=FortConversationTaskNode_CompleteQuestObjectiveByBackendName
  properties {
   objectiveBackendName=quest_s25_stories_p02_q03_02_obj3
   questDefinition=&quest_s25_stories_p02_q03_02
  }
 }
 ConversationEntryPointNode_0 {
  id=#1
  class=ConversationEntryPointNode
  outputs [#15]
  properties {entryTag {tagName=Conversation.Entry.Season25.P01.Q01.Turk}}
 }
 Speech_C_3 {
  id=#15
  class=Speech_C
  outputs [#16]
  properties {
   softTaskIcon=&T_UI_NPC_MainQuest
   generalConfig {defaultMessage="You're, uh... You're looking for the Coral Buddies? They're not here, man."}
  }
  sub_nodes [
   ConversationChoiceNode_8 {
    class=ConversationChoiceNode
    properties {
     defaultChoiceDisplayText=Confront Turk
     choiceTags {gameplayTags [{tagName=Conversation.Choice.AutoSelectIfValid}, {tagName=Conversation.Choice.Category.MainQuestline}]}
    }
   }
   FortConversationRequirement_ControllerRequirement_6 {
    class=FortConversationRequirement_ControllerRequirement
    properties {
     nodeName=Require: Has Completed Quest
     requirement {
      class=FortControllerRequirement_HasCompletedQuest
      properties {bInverseCondition=true, questDefinition=&quest_s25_stories_p02_q03_01}
     }
    }
   }
   FortConversationRequirement_ControllerRequirement_7 {
    class=FortConversationRequirement_ControllerRequirement
    properties {
     nodeName=Require: Is Quest Active
     requirement {
      class=FortControllerRequirement_IsQuestActive
      properties {questDefinition=&quest_s25_stories_p02_q03_02}
     }
    }
   }
   FortConversationRequirement_ControllerRequirement_8 {
    class=FortConversationRequirement_ControllerRequirement
    properties {
     nodeName=Require: NPC Has Conversation Tag
     participantId {tagName=Conversation.Participant.NPC}
     requirement {class=FortControllerRequirement_NPCHasConversationEntryTag, properties {queryTag {tagName=Conversation.Entry.Turk}}}
    }
   }
  ]
 }
 Speech_C_28 {
  id=#16
  class=Speech_C
  outputs [#17]
  properties {generalConfig {defaultMessage=We are definitely not selling all those Coral Buddies to a wealthy collector. Nope.}}
  sub_nodes [ConversationChoiceNode_9 {class=ConversationChoiceNode, properties {defaultChoiceDisplayText=Continue}}]
 }
 Speech_C_13 {
  id=#17
  class=Speech_C
  outputs [#18]
  properties {generalConfig {defaultMessage=You'd better leave Berg Barge if you know what's good for you.}}
  sub_nodes [ConversationChoiceNode_14 {
    class=ConversationChoiceNode
    properties {defaultChoiceDisplayText=Continue, choiceTags {gameplayTags [{tagName=Athena.Quests.S25.Conversations.Turk.Choice.Q01}]}}
   }]
 }
 FortConversationTaskNode_CompleteQuestObjectiveByBackendName_5 {
  id=#18
  class=FortConversationTaskNode_CompleteQuestObjectiveByBackendName
  properties {
   objectiveBackendName=quest_s25_stories_p02_q03_02_obj2
   questDefinition=&quest_s25_stories_p02_q03_02
  }
 }
]
